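# Physics asset for the F_MED_HeistSleek_NPC female medium character body, covering only the head and the left and right hood bones.
# Each Dyn_Master_Hood bone has a small sideways-rotated capsule, and the head has a tiny capsule as the shared anchor body.
# Two mirrored constraints tie the hood bodies to the head with narrow swing limits and almost no twist, restricting how far the hood can move.
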PhysicsAsset /BRCosmetics/Characters/Player/Female/Medium/Bodies/F_MED_HeistSleek_NPC/Meshes/F_MED_HeistSleek_NPC_Physics
bodies {
 Dyn_Master_Hood_l [{shapeType=Capsule, length=5, radius=3, shapeName=Dyn_Master_Hood_l_capsule, center=<2 0 0>, rotation=<rot 90 0 0>}]
 Dyn_Master_Hood_r [{shapeType=Capsule, length=5, radius=3, shapeName=Dyn_Master_Hood_r_capsule, center=<2 0 0>, rotation=<rot 90 0 0>}]
 head [{shapeType=Capsule, length=1, radius=1, shapeName=head_capsule}]
}
constraints [
 {bone1Name=Dyn_Master_Hood_l, bone2Name=head, swing1LimitDegrees=20, swing2LimitDegrees=5, twistLimitDegrees=1}
 {bone1Name=Dyn_Master_Hood_r, bone2Name=head, swing1LimitDegrees=20, swing2LimitDegrees=5, twistLimitDegrees=1}
]
